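# Niagara ribbon trail effect for the TikiTorch player emote.
# Three emitters continuously spawn short-lived ribbon particles, layering a grey band, a wider dark translucent band and a thin, slightly offset bright streak.
# The grey and bright layers multiply their spawn color by a time-of-day tint.
# Only the dark layer fades alpha over life; the ScaleColor fades on the other two are disabled.
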NiagaraSystem /BRCosmetics/Animation/Game/MainPlayer/Emotes/TikiTorch/FX/NS_Emote_TikiTorch_Trail
dependencies {usedDataInterfaces [&NiagaraDataInterfaceCurve], usedDynamicInputs [], usedModules [&SystemState]}
emitters {
 Empty {
  emitterUpdateScript [
   &EmitterState()
   &SpawnRate(SpawnRate=64, Use Spawn Probability=false)
  ]
  particleSpawnScript [
   &InitializeParticle(
    Color=<rgba 0.4167 0.4167 0.4167 1>
    Lifetime=0.25
    Ribbon Width=1.955
    Ribbon Width Mode=::NewEnumerator1
    UsePositionOffset=false
    Write Material Random=true
   )
   &Color(Color=&MultiplyLinearColor(
     Linear Color=$Particles.Color
     Scale Factor=&MakeVector4_FromLinearColor(LinearColor=&FN_Select_Color_from_Time_Of_Day(Morning=<rgba 1 0.735 0.6951 1>, Day=<rgba 1 1 1 1>, Evening=<rgba 1 1 1 1>, Night=<rgba 0.3684 0.3928 0.625 1>))
    ))
  ]
  particleUpdateScript [
   &ParticleState()
   &ScaleColor(
    Scale Alpha=&FloatFromCurve(FloatCurve=DataInterface() {propertyValues {Curve {keys [
         {interpMode=RCIM_Cubic, leaveTangent=-0.1946, tangentMode=RCTM_Break, time=0.8, value=0.9}
         {arriveTangent=-7.5236, tangentMode=RCTM_User, time=1, value=0}
        ]}}})
    ScaleA=true
    ScaleRGB=true
   ) {enabled=false}
  ]
  rendererClasses [&NiagaraRibbonRendererProperties]
  renderers [{rendererClass=&NiagaraRibbonRendererProperties}]
 }
 Empty001 {
  emitterUpdateScript [
   &EmitterState()
   &SpawnRate(SpawnRate=64, Use Spawn Probability=false)
  ]
  particleSpawnScript [&InitializeParticle(
    Color=<rgba 0.0156 0.0156 0.0156 0.65>
    Lifetime=0.25
    Ribbon Width=2.53
    Ribbon Width Mode=::NewEnumerator1
    UsePositionOffset=false
    Write Material Random=true
   )]
  particleUpdateScript [
   &ParticleState()
   &ScaleColor(
    Scale Alpha=&FloatFromCurve(FloatCurve=DataInterface() {propertyValues {Curve {keys [<at 0.6 1>,<at 1 0>]}}})
    ScaleA=true
    ScaleRGB=true
   )
  ]
  rendererClasses [&NiagaraRibbonRendererProperties]
  renderers [{rendererClass=&NiagaraRibbonRendererProperties}]
 }
 Empty002 {
  emitterUpdateScript [
   &EmitterState()
   &SpawnRate(SpawnRate=64, Use Spawn Probability=false)
  ]
  particleSpawnScript [
   &InitializeParticle(
    Color=<rgba 3 3 3 0.6>
    Lifetime=0.25
    Position Offset=<0 0 0.2>
    Ribbon Width=0.46
    Ribbon Width Mode=::NewEnumerator1
    UsePositionOffset=true
    Write Material Random=true
   )
   &Color(Color=&MultiplyLinearColor(
     Linear Color=$Particles.Color
     Scale Factor=&MakeVector4_FromLinearColor(LinearColor=&FN_Select_Color_from_Time_Of_Day(Morning=<rgba 1 0.8209 0.801 1>, Day=<rgba 1 1 1 1>, Evening=<rgba 1 1 1 1>, Night=<rgba 0.3684 0.3928 0.625 1>))
    ))
  ]
  particleUpdateScript [
   &ParticleState()
   &ScaleColor(
    Scale Alpha=&FloatFromCurve(FloatCurve=DataInterface() {propertyValues {Curve {keys [
         {interpMode=RCIM_Cubic, leaveTangent=-0.1946, tangentMode=RCTM_Break, time=0.8, value=0.9}
         {arriveTangent=-7.5236, tangentMode=RCTM_User, time=1, value=0}
        ]}}})
    ScaleA=true
    ScaleRGB=true
   ) {enabled=false}
  ]
  rendererClasses [&NiagaraRibbonRendererProperties]
  renderers [{rendererClass=&NiagaraRibbonRendererProperties}]
 }
}
summary {
 systemName=NS_Emote_TikiTorch_Trail
 emitters [
  {emitterName=Empty, rendererClasses [&NiagaraRibbonRendererProperties]}
  {emitterName=Empty001, rendererClasses [&NiagaraRibbonRendererProperties]}
  {emitterName=Empty002, rendererClasses [&NiagaraRibbonRendererProperties]}
 ]
 userVariables []
}
